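# Physics asset for the dangling gear on the Cavern Armored medium male body: hip bags, a back flap and two grenades with rings.
# Each gear piece hangs from its own tiny MSTR anchor body, which joins the pelvis with free swing and twist.
# Angular limits on the anchor-to-gear joints shape the sway, and each ring chains off its grenade with its own limits.
# Large pelvis and thigh capsules provide body volumes around the gear.
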PhysicsAsset /BRCosmetics/Characters/Player/Male/Medium/Bodies/M_MED_Cavern_Armored/Meshes/M_MED_Cavern_Armored_dyn_bags
bodies {
 dyn_bag_00_l [{shapeType=Capsule, length=5.875, radius=3.75, center=<5 0 4>, rotation=<rot 90 0 0>}]
 dyn_bag_00_l_MSTR [{shapeType=Capsule, length=0.505, radius=0.505, rotation=<rot 0 -90 0>}]
 dyn_bag_01_l [{shapeType=Capsule, length=6.25, radius=3.375, center=<5 0 3.5>, rotation=<rot 90 0 0>}]
 dyn_bag_01_l_MSTR [{shapeType=Capsule, length=0.505, radius=0.505, rotation=<rot 0 -90 0>}]
 dyn_bag_02_l [{shapeType=Capsule, length=5.75, radius=3.875, center=<5 0 4>, rotation=<rot 90 0 0>}]
 dyn_bag_02_l_MSTR [{shapeType=Capsule, length=0.505, radius=0.505, rotation=<rot 0 -90 0>}]
 dyn_bag_02_r [{shapeType=Capsule, length=3.625, radius=4.375, center=<5 0 4>, rotation=<rot 90 0 0>}]
 dyn_bag_02_r_MSTR [{shapeType=Capsule, length=0.505, radius=0.505, rotation=<rot 0 -90 0>}]
 dyn_flap_bk [{shapeType=Capsule, length=8.25, radius=5.75, center=<10 0 3>, rotation=<rot 0 0 90>}]
 dyn_flap_bk_MSTR [{shapeType=Capsule, length=0.505, radius=0.505, rotation=<rot 0 -90 0>}]
 dyn_grenades_01 [{shapeType=Capsule, length=8.8678, radius=3.7521, center=<3 0 4.5>, rotation=<rot 0 85.7892 -87.8296>}]
 dyn_grenades_01_MSTR [{shapeType=Capsule, length=0.505, radius=0.505, rotation=<rot 0 -90 0>}]
 dyn_grenades_02 [{shapeType=Capsule, length=7.7428, radius=3.6842, center=<2.9896 0.001 4>, rotation=<rot 0 84.9947 -88.8985>}]
 dyn_grenades_02_MSTR [{shapeType=Capsule, length=0.505, radius=0.505, rotation=<rot 0 -90 0>}]
 dyn_ring_01 [{shapeType=Capsule, length=1, radius=1, center=<1 0 0>, rotation=<rot 90 0 0>}]
 dyn_ring_02 [{shapeType=Capsule, length=1, radius=1, center=<1 0 0>, rotation=<rot 90 0 0>}]
 pelvis [{shapeType=Capsule, length=6.9474, radius=11.8815, center=<-0.0001 -1.4047 0.0001>}]
 thigh_l [{shapeType=Capsule, length=27.7764, radius=11.2583, center=<-11.9782 -1.1753 1.9351>, rotation=<rot 88.1666 179.9999 179.9999>}]
 thigh_r [{shapeType=Capsule, length=27.7764, radius=11.2583, center=<11.9782 1.1753 -1.9351>, rotation=<rot 88.1666 179.9999 -0.0001>}]
}
constraints [
 {
  bone1Name=dyn_bag_02_r_MSTR
  bone2Name=pelvis
  swing1LimitDegrees=45
  swing1Motion=Free
  swing2LimitDegrees=45
  swing2Motion=Free
  twistLimitDegrees=45
  twistMotion=Free
 }
 {bone1Name=dyn_bag_02_r, bone2Name=dyn_bag_02_r_MSTR, swing1LimitDegrees=10, swing2LimitDegrees=20, twistLimitDegrees=2}
 {
  bone1Name=dyn_bag_01_l_MSTR
  bone2Name=pelvis
  swing1LimitDegrees=45
  swing1Motion=Free
  swing2LimitDegrees=45
  swing2Motion=Free
  twistLimitDegrees=45
  twistMotion=Free
 }
 {bone1Name=dyn_bag_01_l, bone2Name=dyn_bag_01_l_MSTR, swing1LimitDegrees=5, swing2LimitDegrees=25, twistLimitDegrees=2}
 {
  bone1Name=dyn_bag_02_l_MSTR
  bone2Name=pelvis
  swing1LimitDegrees=45
  swing1Motion=Free
  swing2LimitDegrees=45
  swing2Motion=Free
  twistLimitDegrees=45
  twistMotion=Free
 }
 {bone1Name=dyn_bag_02_l, bone2Name=dyn_bag_02_l_MSTR, swing1LimitDegrees=10, swing2LimitDegrees=20, twistLimitDegrees=2}
 {
  bone1Name=thigh_l
  bone2Name=pelvis
  swing1LimitDegrees=45
  swing1Motion=Free
  swing2LimitDegrees=45
  swing2Motion=Free
  twistLimitDegrees=45
  twistMotion=Free
 }
 {
  bone1Name=dyn_grenades_01_MSTR
  bone2Name=pelvis
  swing1LimitDegrees=45
  swing1Motion=Free
  swing2LimitDegrees=45
  swing2Motion=Free
  twistLimitDegrees=45
  twistMotion=Free
 }
 {bone1Name=dyn_grenades_01, bone2Name=dyn_grenades_01_MSTR, swing1LimitDegrees=25, swing2LimitDegrees=25, twistLimitDegrees=10}
 {bone1Name=dyn_ring_01, bone2Name=dyn_grenades_01, swing1LimitDegrees=5, swing2LimitDegrees=30, twistLimitDegrees=2}
 {
  bone1Name=dyn_grenades_02_MSTR
  bone2Name=pelvis
  swing1LimitDegrees=45
  swing1Motion=Free
  swing2LimitDegrees=45
  swing2Motion=Free
  twistLimitDegrees=45
  twistMotion=Free
 }
 {bone1Name=dyn_grenades_02, bone2Name=dyn_grenades_02_MSTR, swing1LimitDegrees=25, swing2LimitDegrees=25, twistLimitDegrees=10}
 {bone1Name=dyn_ring_02, bone2Name=dyn_grenades_02, swing1LimitDegrees=5, swing2LimitDegrees=30, twistLimitDegrees=2}
 {
  bone1Name=dyn_bag_00_l_MSTR
  bone2Name=pelvis
  swing1LimitDegrees=45
  swing1Motion=Free
  swing2LimitDegrees=45
  swing2Motion=Free
  twistLimitDegrees=45
  twistMotion=Free
 }
 {bone1Name=dyn_bag_00_l, bone2Name=dyn_bag_00_l_MSTR, swing1LimitDegrees=10, swing2LimitDegrees=20, twistLimitDegrees=2}
 {
  bone1Name=dyn_flap_bk_MSTR
  bone2Name=pelvis
  swing1LimitDegrees=45
  swing1Motion=Free
  swing2LimitDegrees=45
  swing2Motion=Free
  twistLimitDegrees=45
  twistMotion=Free
 }
 {bone1Name=dyn_flap_bk, bone2Name=dyn_flap_bk_MSTR, swing1LimitDegrees=2, swing2LimitDegrees=30, twistLimitDegrees=5}
 {
  bone1Name=thigh_r
  bone2Name=pelvis
  swing1LimitDegrees=45
  swing1Motion=Free
  swing2LimitDegrees=45
  swing2Motion=Free
  twistLimitDegrees=45
  twistMotion=Free
 }
]
