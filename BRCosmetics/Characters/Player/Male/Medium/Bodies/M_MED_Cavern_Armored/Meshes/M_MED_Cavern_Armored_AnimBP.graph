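# Animation Blueprint for the Cavern Armored male medium body that follows the attached parent's pose and adds deformation and gear motion.
# The copied pose runs through twist and sculpter Control Rigs, then pelvis/spine constraints on bags and grenades.
# Thigh and head rotations, remapped each update, weight additive offsets on that gear and the back collar before pelvis-space RigidBody simulation.
# Side-to-side collar chains are disconnected from the output.
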AnimBlueprint /BRCosmetics/Characters/Player/Male/Medium/Bodies/M_MED_Cavern_Armored/Meshes/M_MED_Cavern_Armored_AnimBP
target_skeleton=&/M_MED_Cavern_Armored_Skeleton
graphs [
 AnimGraph {nodes [
   n1 = Get head rot l() -> (head rot l:double)
   n2 = "Transform (Modify) Bone - Bone: dyn_collar_bk_MSTR"(Alpha=@n1, Translation="0,0,0", Rotation="0,-30,0", Scale="1,1,1") {
    rotationMode=BMM_Additive
    rotationSpace=BCS_BoneSpace
    scaleMode=BMM_Additive
    scaleSpace=BCS_BoneSpace
    translationMode=BMM_Additive
    translationSpace=BCS_BoneSpace
    alphaBoolBlend {blendOption=Linear}
    boneToModify {boneName=dyn_collar_bk_MSTR}
    rotation {yaw=-30}
   }
   "Transform (Modify) Bone - Bone: dyn_collar_l_MSTR"(ComponentPose=@n2, Alpha=@n1, Translation="0,0,0", Rotation="20,-20,0", Scale="0.6,1.000,1.000") {
    rotationMode=BMM_Additive
    rotationSpace=BCS_BoneSpace
    scaleMode=BMM_Additive
    scaleSpace=BCS_BoneSpace
    translationMode=BMM_Additive
    translationSpace=BCS_BoneSpace
    alphaBoolBlend {blendOption=Linear}
    boneToModify {boneName=dyn_collar_l_MSTR}
    rotation {pitch=20, yaw=-20}
    scale {x=0.6}
   }
   n3 = "Transform (Modify) Bone - Bone: dyn_collar_r_MSTR"(Alpha=0, Translation="0,0,0", Rotation="5,0,0", Scale="0.7,1.000,1.000") {
    rotationMode=BMM_Additive
    rotationSpace=BCS_BoneSpace
    scaleMode=BMM_Additive
    scaleSpace=BCS_BoneSpace
    translationMode=BMM_Additive
    translationSpace=BCS_BoneSpace
    alphaBoolBlend {blendOption=Linear}
    boneToModify {boneName=dyn_collar_r_MSTR}
    rotation {pitch=5}
    scale {x=0.7}
   }
   n4 = "Transform (Modify) Bone - Bone: dyn_collar_l_MSTR"(ComponentPose=@n3, Alpha=0, Translation="0,0,0", Rotation="5,0,0", Scale="0.7,1,1") {
    rotationMode=BMM_Additive
    rotationSpace=BCS_BoneSpace
    scaleMode=BMM_Additive
    scaleSpace=BCS_BoneSpace
    translationMode=BMM_Additive
    translationSpace=BCS_BoneSpace
    alphaBoolBlend {blendOption=Linear}
    boneToModify {boneName=dyn_collar_l_MSTR}
    rotation {pitch=5}
    scale {x=0.7}
   }
   n5 = Get head rot r() -> (head rot r:double)
   n6 = "Transform (Modify) Bone - Bone: dyn_collar_bk_MSTR"(ComponentPose=@n4, Alpha=@n5, Translation="0,0,0", Rotation="0,30,0", Scale="1,1,1") {
    rotationMode=BMM_Additive
    rotationSpace=BCS_BoneSpace
    scaleMode=BMM_Additive
    scaleSpace=BCS_BoneSpace
    translationMode=BMM_Additive
    translationSpace=BCS_BoneSpace
    alphaBoolBlend {blendOption=Linear}
    boneToModify {boneName=dyn_collar_bk_MSTR}
    rotation {yaw=30}
   }
   "Transform (Modify) Bone - Bone: dyn_collar_r_MSTR"(ComponentPose=@n6, Alpha=@n5, Translation="0,0,0", Rotation="20,20,0", Scale="0.6,1.000,1.000") {
    rotationMode=BMM_Additive
    rotationSpace=BCS_BoneSpace
    scaleMode=BMM_Additive
    scaleSpace=BCS_BoneSpace
    translationMode=BMM_Additive
    translationSpace=BCS_BoneSpace
    alphaBoolBlend {blendOption=Linear}
    boneToModify {boneName=dyn_collar_r_MSTR}
    rotation {pitch=20, yaw=20}
    scale {x=0.6}
   }
   n7 = Control Rig(Source=Copy Pose From Mesh() {bUseAttachedParent=true}) {
    alphaBoolBlend {blendOption=Linear}
    controlRigAssetReference {blueprintRigClass=&/M_MED_Cavern_Armored_Deform_CtrlRig_twists.M_MED_Cavern_Armored_Deform_CtrlRig_twists_C}
   }
   n8 = Control Rig!(
    Source=@n7
    Alpha
    KneeRot_fr_l_114deg:Vector
    KneeTrans_fr_l_114deg:Vector="0.258,-3.591,0.009"
    KneeScale_fr_l_114deg:Vector="1,1,1"
    KneeScale_fr_l_130deg:Vector="1,1,1"
    KneeRot_fr_l_130deg:Vector="0,0,-51.2"
    KneeTrans_fr_l_130deg:Vector="0.258,-3.591,0.009"
   ) -> (Pose) {
    alphaBoolBlend {blendOption=Linear}
    controlRigAssetReference {blueprintRigClass=&/M_MED_Cavern_Armored_Deform_CtrlRig_sculpters.M_MED_Cavern_Armored_Deform_CtrlRig_sculpters_C}
   }
   n9 = "Constraint - dyn_grenades_01_MSTR (PPPP)"!(
    ComponentPose=Local To Component(LocalPose=@n8)
    Alpha=1
    ConstraintWeights_0=0.6673
    ConstraintWeights_1:float=0.0915
    ConstraintWeights_2:float=0.0282
    ConstraintWeights_3:float=0.2129
   ) -> (Pose) {
    alphaBoolBlend {blendOption=Linear}
    boneToModify {boneName=dyn_grenades_01_MSTR}
    constraintSetup [{targetBone {boneName=pelvis}}, {targetBone {boneName=spine_02}}, {targetBone {boneName=spine_03}}, {targetBone {boneName=spine_01}}]
    constraintWeights [1,1,1,1]
   }
   n10 = "Constraint - dyn_grenades_02_MSTR (PPPP)"!(
    ComponentPose=@n9
    Alpha=1
    ConstraintWeights_0=0.6673
    ConstraintWeights_1:float=0.0915
    ConstraintWeights_2:float=0.0282
    ConstraintWeights_3:float=0.2129
   ) -> (Pose) {
    alphaBoolBlend {blendOption=Linear}
    boneToModify {boneName=dyn_grenades_02_MSTR}
    constraintSetup [{targetBone {boneName=pelvis}}, {targetBone {boneName=spine_02}}, {targetBone {boneName=spine_03}}, {targetBone {boneName=spine_01}}]
    constraintWeights [1,1,1,1]
   }
   n11 = "Constraint - dyn_bag_02_r_MSTR (PPPP)"!(
    ComponentPose=@n10
    Alpha=1
    ConstraintWeights_0=0.6673
    ConstraintWeights_1:float=0.0915
    ConstraintWeights_2:float=0.0282
    ConstraintWeights_3:float=0.2129
   ) -> (Pose) {
    alphaBoolBlend {blendOption=Linear}
    boneToModify {boneName=dyn_bag_02_r_MSTR}
    constraintSetup [{targetBone {boneName=pelvis}}, {targetBone {boneName=spine_02}}, {targetBone {boneName=spine_03}}, {targetBone {boneName=spine_01}}]
    constraintWeights [1,1,1,1]
   }
   n12 = "Constraint - dyn_bag_00_l_MSTR (PPPP)"!(
    ComponentPose=@n11
    Alpha=1
    ConstraintWeights_0=0.6673
    ConstraintWeights_1:float=0.0915
    ConstraintWeights_2:float=0.0282
    ConstraintWeights_3:float=0.2129
   ) -> (Pose) {
    alphaBoolBlend {blendOption=Linear}
    boneToModify {boneName=dyn_bag_00_l_MSTR}
    constraintSetup [{targetBone {boneName=pelvis}}, {targetBone {boneName=spine_02}}, {targetBone {boneName=spine_03}}, {targetBone {boneName=spine_01}}]
    constraintWeights [1,1,1,1]
   }
   n13 = "Constraint - dyn_bag_01_l_MSTR (PPPP)"!(
    ComponentPose=@n12
    Alpha=1
    ConstraintWeights_0=0.6673
    ConstraintWeights_1:float=0.0915
    ConstraintWeights_2:float=0.0282
    ConstraintWeights_3:float=0.2129
   ) -> (Pose) {
    alphaBoolBlend {blendOption=Linear}
    boneToModify {boneName=dyn_bag_01_l_MSTR}
    constraintSetup [{targetBone {boneName=pelvis}}, {targetBone {boneName=spine_02}}, {targetBone {boneName=spine_03}}, {targetBone {boneName=spine_01}}]
    constraintWeights [1,1,1,1]
   }
   n14 = "Constraint - dyn_bag_02_l_MSTR (PPPP)"!(
    ComponentPose=@n13
    Alpha=1
    ConstraintWeights_0=0.6673
    ConstraintWeights_1:float=0.0915
    ConstraintWeights_2:float=0.0282
    ConstraintWeights_3:float=0.2129
   ) -> (Pose) {
    alphaBoolBlend {blendOption=Linear}
    boneToModify {boneName=dyn_bag_02_l_MSTR}
    constraintSetup [{targetBone {boneName=pelvis}}, {targetBone {boneName=spine_02}}, {targetBone {boneName=spine_03}}, {targetBone {boneName=spine_01}}]
    constraintWeights [1,1,1,1]
   }
   n15 = Get thigh_rot_up_l() -> (thigh_rot_up_l:double)
   n16 = "Transform (Modify) Bone - Bone: dyn_bag_02_l_MSTR"(ComponentPose=@n14, Alpha=@n15, Translation="0.203,-0.752,-0.350", Rotation="9.601,0.000,-14.568", Scale="1,1,1") {
    rotationMode=BMM_Additive
    rotationSpace=BCS_BoneSpace
    translationMode=BMM_Additive
    translationSpace=BCS_BoneSpace
    alphaBoolBlend {blendOption=Linear}
    boneToModify {boneName=dyn_bag_02_l_MSTR}
    rotation {pitch=9.601, roll=-14.5683}
    translation=<0.203 -0.752 -0.35>
   }
   n17 = "Transform (Modify) Bone - Bone: dyn_bag_01_l_MSTR"(ComponentPose=@n16, Alpha=@n15, Translation="-0.814,-1.588,0.180", Rotation="19.329,-2.391,-13.706", Scale="1,1,1") {
    rotationMode=BMM_Additive
    rotationSpace=BCS_BoneSpace
    translationMode=BMM_Additive
    translationSpace=BCS_BoneSpace
    alphaBoolBlend {blendOption=Linear}
    boneToModify {boneName=dyn_bag_01_l_MSTR}
    rotation=<rot 19.329 -2.391 -13.706>
    translation=<-0.814 -1.588 0.18>
   }
   n18 = "Transform (Modify) Bone - Bone: dyn_bag_00_l_MSTR"(ComponentPose=@n17, Alpha=@n15, Translation="-1.112,-1.861,0.196", Rotation="16.949,-9.728,-29.025", Scale="1,1,1") {
    rotationMode=BMM_Additive
    rotationSpace=BCS_BoneSpace
    translationMode=BMM_Additive
    translationSpace=BCS_BoneSpace
    alphaBoolBlend {blendOption=Linear}
    boneToModify {boneName=dyn_bag_00_l_MSTR}
    rotation=<rot 16.9489 -9.7284 -29.025>
    translation=<-1.112 -1.861 0.196>
   }
   n19 = Get thigh_rot_up_r() -> (thigh_rot_up_r:double)
   n20 = "Transform (Modify) Bone - Bone: dyn_grenades_01_MSTR"(ComponentPose=@n18, Alpha=@n19, Translation="-0.130,-0.584,1.458", Rotation="-3.211,3.834,39.892", Scale="1,1,1") {
    rotationMode=BMM_Additive
    rotationSpace=BCS_BoneSpace
    translationMode=BMM_Additive
    translationSpace=BCS_BoneSpace
    alphaBoolBlend {blendOption=Linear}
    boneToModify {boneName=dyn_grenades_01_MSTR}
    rotation=<rot -3.211 3.834 39.892>
    translation=<-0.13 -0.5838 1.4577>
   }
   n21 = "Transform (Modify) Bone - Bone: dyn_grenades_02_MSTR"(ComponentPose=@n20, Alpha=@n19, Translation="1.109,1.137,1.108", Rotation="20.704,22.207,49.108", Scale="1,1,1") {
    rotationMode=BMM_Additive
    rotationSpace=BCS_BoneSpace
    translationMode=BMM_Additive
    translationSpace=BCS_BoneSpace
    alphaBoolBlend {blendOption=Linear}
    boneToModify {boneName=dyn_grenades_02_MSTR}
    rotation=<rot 20.704 22.2067 49.108>
    translation=<1.109 1.1368 1.1082>
   }
   n22 = "Transform (Modify) Bone - Bone: dyn_bag_02_r_MSTR"(ComponentPose=@n21, Alpha=@n19, Translation="-0.097,1.047,-0.224", Rotation="6.904,12.331,16.630", Scale="1,1,1") {
    rotationMode=BMM_Additive
    rotationSpace=BCS_BoneSpace
    translationMode=BMM_Additive
    translationSpace=BCS_BoneSpace
    alphaBoolBlend {blendOption=Linear}
    boneToModify {boneName=dyn_bag_02_r_MSTR}
    rotation=<rot 6.904 12.3311 16.6301>
    translation=<-0.097 1.047 -0.224>
   }
   n23 = "Transform (Modify) Bone - Bone: dyn_collar_bk_MSTR"(ComponentPose=@n22, Alpha=Get head rot back() -> (head rot back:double), Translation="0,0,0", Rotation="10,0.000,0.000", Scale="1,1,1") {
    rotationMode=BMM_Additive
    rotationSpace=BCS_BoneSpace
    scaleMode=BMM_Additive
    scaleSpace=BCS_BoneSpace
    translationMode=BMM_Additive
    translationSpace=BCS_BoneSpace
    alphaBoolBlend {blendOption=Linear}
    boneToModify {boneName=dyn_collar_bk_MSTR}
    rotation {pitch=10}
   }
   n24 = RigidBody(ComponentPose=@n23) {
    bClampLinearTranslationLimitToRefPose=true
    bOverrideWorldGravity=true
    overridePhysicsAsset=&/M_MED_Cavern_Armored_dyn_bags
    simulationSpace=BaseBoneSpace
    alphaBoolBlend {blendOption=Linear}
    baseBoneRef {boneName=pelvis}
    componentAppliedLinearAccClamp=<3000 3000 3000>
    componentLinearAccScale=<0.32 0.35 0.35>
    componentLinearVelScale=<0.3 0.3 0.3>
    overrideWorldGravity {z=-980}
   }
   Output Pose(Result=Component To Local(ComponentPose=@n24)) {layerGroup=None}
  ]}
 EventGraph {nodes [
   Event BlueprintUpdateAnimation()
   > n1 = Sequence()
   n2 = BreakTransform(InTransform=GetDeltaTransformFromRefPose(self=GetOwnerSkeletalMeshComponent(), BoneName=thigh_r, BaseName=pelvis))
   > Set thigh_rot_up_r(thigh_rot_up_r:double=MapRangeClamped(Value=BreakRotator(InRot=@n2.Rotation).Yaw, InRangeA=-15, InRangeB=-100)) -> (Output_Get:double)
   n3 = BreakTransform(InTransform=GetDeltaTransformFromRefPose(self=GetOwnerSkeletalMeshComponent(), BoneName=thigh_l, BaseName=pelvis))
   > Set thigh_rot_up_l(thigh_rot_up_l:double=MapRangeClamped(Value=BreakRotator(InRot=@n3.Rotation).Yaw, InRangeA=-45, InRangeB=-135)) -> (Output_Get:double)
   n4 = GetOwnerSkeletalMeshComponent()
   n5 = BreakRotator(InRot=BreakTransform(InTransform=GetDeltaTransformFromRefPose(self=@n4, BoneName=head, BaseName=neck_01)).Rotation)
   Set head rot back(execute=@n1.then_1, head rot back:double=MapRangeClamped(Value=@n5.Yaw, InRangeA=5, InRangeB=-35)) -> (Output_Get:double)
   > Set head rot r(head rot r:double=MapRangeClamped(Value=@n5.Pitch, InRangeA=0, InRangeB=20)) -> (Output_Get:double)
   > Set head rot l(head rot l:double=MapRangeClamped(Value=@n5.Pitch, InRangeA=0, InRangeB=-20)) -> (Output_Get:double)
  ]}
]
variables [
 head rot back {type=number}
 head rot l {type=number}
 head rot r {type=number}
 thigh_rot_up_l {type=number}
 thigh_rot_up_r {type=number}
]
